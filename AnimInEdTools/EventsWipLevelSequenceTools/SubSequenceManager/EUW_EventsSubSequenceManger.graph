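# Editor utility widget that toggles audio, camera, gameplay, lighting, tech art and VFX subsequences in a Level Sequence.
# Selecting a sequence reads project, subfolder and shot from metadata (or its path), then sets checkboxes from matching sub-track sections.
# Set Sub Tracks applies checkbox states as section active flags; "isSubSequenceMute" true means active.
# Checked but missing tracks are created under a SubSequences folder, loading assets from paths built via the project's AnimatableCreature.
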EditorUtilityWidgetBlueprint /AnimInEdTools/EventsWipLevelSequenceTools/SubSequenceManager/EUW_EventsSubSequenceManger
graphs [
 AddMissingSubSequence {nodes [
   AddMissingSubSequence!() -> (then, MissingSubSequence:String) {is=title K2Node_FunctionEntry}
   > Set FolderRef(FolderRef:Movie Scene Folder Object Reference=CheckAddFolder!() -> (SubFolderRef:Movie Scene Folder Object Reference)) -> (Output_Get:Movie Scene Folder Object Reference)
   > n1 = AddTrack!(
    execute
    Sequence:Movie Scene Sequence Object Reference=Get Level Sequence() -> (Level Sequence:Level Sequence Object Reference)
    TrackType:Movie Scene Track Class Reference=/Script/MovieScene.MovieSceneSubTrack
   ) -> (then, ReturnValue:Movie Scene Sub Track Object Reference)
   > SetDisplayName!(
    execute
    Track:Movie Scene Track Object Reference=@n1
    InName:"Text (by ref)"="To Text (String)"(InString=Get MissingSubSequence() -> (MissingSubSequence:String))
   ) -> (then)
   > AddChildTrack!(
    execute
    Folder:Movie Scene Folder Object Reference=Get FolderRef() -> (FolderRef:Movie Scene Folder Object Reference)
    InTrack:Movie Scene Track Object Reference=@n1
   ) -> (then, ReturnValue:bool)
   > n2 = AddSection!(execute, Track:Movie Scene Track Object Reference=@n1) -> (then, ReturnValue:Movie Scene Section Object Reference)
   > n3 = GetPlaybackRange!(execute, Sequence:Movie Scene Sequence Object Reference=Get Level Sequence()) -> (then, ReturnValue_bHasStart:bool=~, ReturnValue_bHasEnd:bool=~, ReturnValue_InclusiveStart:int, ReturnValue_ExclusiveEnd:int)
   > SetRange!(execute, Section:Movie Scene Section Object Reference=@n2, StartFrame:int=0, EndFrame:int=int + int(A=@n3.ReturnValue_ExclusiveEnd)) -> (then)
   > n4 = Cast To MovieSceneSubSection!(execute, Object=@n2) -> (then, CastFailed, AsMovie Scene Sub Section:Movie Scene Sub Section Object Reference) {is=title K2Node_DynamicCast}
   n5 = Append!(
    A=Get ProjectName() -> (ProjectName:String)
    B=_
    C:String=Get SubFolderName() -> (SubFolderName:String)
    D:String=Get ShotNumber() -> (ShotNumber:String)
    E:String=_
    F:String=Get MissingSubSequence()
   ) -> (ReturnValue)
   n6 = Get AnimatableCreature() -> (AnimatableCreature:Animatable Creature Object Reference)
   n7 = Get SubFolderName()
   > n8 = LoadAsset(AssetPath=Append!(
     A=Get AnimMontage Path!(self:Animatable Creature Object Reference=@n6) -> (AnimMontage Path:String)
     B=@n7
     C=/
     D=@n7
     E=Get ShotNumber()
     F=/
     G:String=@n5
     H:String=.
     I:String=@n5
    ) -> (ReturnValue))
   > n9 = Cast To LevelSequence!(execute, Object=@n8) -> (then, CastFailed, AsLevel Sequence:Level Sequence Object Reference) {is=title K2Node_DynamicCast}
   > SetSequence!(execute, self:Movie Scene Sub Section Object Reference=@n4, Sequence:Movie Scene Sequence Object Reference=@n9) -> (then)
   > n10 = GetPlaybackRange(Sequence=@n9) -> (ReturnValue_bHasStart=~, ReturnValue_bHasEnd=~)
   > SetRange(Section=@n2, StartFrame=@n10.ReturnValue_InclusiveStart, EndFrame=@n10.ReturnValue_ExclusiveEnd)
   > Add!(
    execute
    TargetMap:Map of Strings to Booleans=Get ContainsSubSequence() -> (ContainsSubSequence:Map of Strings to Booleans)
    Key:"String (by ref)"=Get MissingSubSequence()
    Value:"Boolean (by ref)"=true
   ) -> (then) {is=title K2Node_CallFunction}
  ]}
 CheckAddFolder {nodes [
   CheckAddFolder() {is=title K2Node_FunctionEntry}
   > n1 = For Each Loop(Array:Array of Movie Scene Folder Object References=GetRootFoldersInSequence!(Sequence:Movie Scene Sequence Object Reference=Get Level Sequence()) -> (ReturnValue:Array of Movie Scene Folder Object References)) -> (Array Element:Movie Scene Folder Object Reference)
   n2 = Branch(execute=@n1.Completed, Condition=Get HasFolder() -> (HasFolder:bool))
   Return Node!(execute=[@n2.then, @n4.then], SubFolderRef:Movie Scene Folder Object Reference=Get FolderRef())
   n3 = AddRootFolderToSequence!(execute=@n2.else, Sequence:Movie Scene Sequence Object Reference=Get Level Sequence(), NewFolderName:String=SubSequences) -> (then, ReturnValue:Movie Scene Folder Object Reference)
   > n4 = Set FolderRef(FolderRef=@n3)
   n5 = GetFolderName!(Folder:Movie Scene Folder Object Reference=@n1) -> (ReturnValue:Name=None)
   Branch(execute=@n1.LoopBody, Condition=Contains(SearchIn="To String (Name)"(InName=@n5), Substring=SubSequences))
   > Set HasFolder(HasFolder:bool=true) -> (Output_Get:bool)
   > Set FolderRef(FolderRef=@n1)
  ]}
 ClearContainsSubSequence {nodes [
   ClearContainsSubSequence() {is=title K2Node_FunctionEntry}
   n1 = Get ContainsSubSequence()
   > Add(TargetMap=@n1, Key=aud)
   > Add(TargetMap=@n1, Key=cam)
   > Add(TargetMap=@n1, Key=gam)
   > Add(TargetMap=@n1, Key=lgt)
   > Add(TargetMap=@n1, Key=tec)
   > Add(TargetMap=@n1, Key=vfx)
   > n2 = For Each Loop(Array:Array of Editor Utility Check Box Object References=Get CheckBoxes() -> (CheckBoxes:Array of Editor Utility Check Box Object References)) -> (Array Element:Editor Utility Check Box Object Reference)
   > Set CheckedState!(execute, CheckedState:ECheckBoxState Enum=Unchecked, self:Check Box Object Reference=@n2) -> (then, Output_Get:ECheckBoxState Enum=Unchecked)
  ]}
 EventGraph {nodes [
   n1 = "On Property Changed (LevelSequencePropertyView)"!() -> (OutputDelegate, then, PropertyName:Name=None) {is=title K2Node_ComponentBoundEvent}
   GetMetaData(execute=[@n1.then, @n12.Completed, @n3.then])
   > GetProjectAC()
   > ClearContainsSubSequence() {is=title K2Node_CallFunction}
   > FindCurrentToggle()
   "On Clicked (RunButton)"() {is=title K2Node_ComponentBoundEvent}
   > OpenEditorForAssets!(
    execute
    self:Asset Editor Subsystem Object Reference=Get AssetEditorSubsystem()
    Assets:Array of Object References=Make Array("[0]":Object Reference=Get Level Sequence()) -> (Array:Array of Object References)
    OpenedMethod:EAssetTypeActivationOpenedMethod Enum=Edit
   ) -> (then, ReturnValue:bool)
   > SetTracks()
   "On Clicked (CurrentLevelSequenceButton)"() {is=title K2Node_ComponentBoundEvent}
   > n2 = GetFocusedLevelSequence!(execute) -> (then, ReturnValue:Level Sequence Object Reference)
   > n3 = Set Level Sequence(Level Sequence:Level Sequence Object Reference=@n2) -> (Output_Get:Level Sequence Object Reference)
   n4 = "On Check State Changed (AudioCheckBox)"!() -> (OutputDelegate, then, bIsChecked:bool) {is=title K2Node_ComponentBoundEvent}
   > Add(TargetMap=Get isSubSequenceMute() -> (isSubSequenceMute:Map of Strings to Booleans), Key=aud, Value=@n4.bIsChecked)
   n5 = "On Check State Changed (CameraCheckBox)"!() -> (OutputDelegate, then, bIsChecked:bool) {is=title K2Node_ComponentBoundEvent}
   > Add(TargetMap=Get isSubSequenceMute(), Key=cam, Value=@n5.bIsChecked)
   n6 = "On Check State Changed (GamePlayCheckBox)"!() -> (OutputDelegate, then, bIsChecked:bool) {is=title K2Node_ComponentBoundEvent}
   > Add(TargetMap=Get isSubSequenceMute(), Key=gam, Value=@n6.bIsChecked)
   n7 = "On Check State Changed (LightingCheckBox)"!() -> (OutputDelegate, then, bIsChecked:bool) {is=title K2Node_ComponentBoundEvent}
   > Add(TargetMap=Get isSubSequenceMute(), Key=lgt, Value=@n7.bIsChecked)
   n8 = "On Check State Changed (TechArtCheckBox)"!() -> (OutputDelegate, then, bIsChecked:bool) {is=title K2Node_ComponentBoundEvent}
   > Add(TargetMap=Get isSubSequenceMute(), Key=tec, Value=@n8.bIsChecked)
   n9 = "On Check State Changed (VFXCheckBox)"!() -> (OutputDelegate, then, bIsChecked:bool) {is=title K2Node_ComponentBoundEvent}
   > Add(TargetMap=Get isSubSequenceMute(), Key=vfx, Value=@n9.bIsChecked)
   Event PreConstruct()
   > SetObject(self=Get LevelSequencePropertyView() -> (LevelSequencePropertyView:Single Property View Object Reference), NewObject=Self-Reference())
   n10 = Get CheckBoxes()
   > Add(
    TargetArray:Array of Editor Utility Check Box Object References=@n10
    NewItem:"Editor Utility Check Box Object Reference (by ref)"=Get AudioCheckBox() -> (AudioCheckBox:Editor Utility Check Box Object Reference)
   ) {is=title K2Node_CallArrayFunction}
   > Add(TargetArray=@n10, NewItem=Get CameraCheckBox() -> (CameraCheckBox:Editor Utility Check Box Object Reference))
   > Add(TargetArray=@n10, NewItem=Get GamePlayCheckBox() -> (GamePlayCheckBox:Editor Utility Check Box Object Reference))
   > Add(TargetArray=@n10, NewItem=Get LightingCheckBox() -> (LightingCheckBox:Editor Utility Check Box Object Reference))
   > Add(TargetArray=@n10, NewItem=Get TechArtCheckBox() -> (TechArtCheckBox:Editor Utility Check Box Object Reference))
   > Add(TargetArray=@n10, NewItem=Get VFXCheckBox() -> (VFXCheckBox:Editor Utility Check Box Object Reference))
   Event Construct()
   > n11 = Set CurrentProjects(CurrentProjects:Array of BP Animation Browser Project Object References=GetProjects!() -> (Projects:Array of BP Animation Browser Project Object References)) -> (Output_Get:Array of BP Animation Browser Project Object References)
   > n12 = For Each Loop(Array:Array of BP Animation Browser Project Object References=@n11) -> (Array Element:BP Animation Browser Project Object Reference)
   > Add(
    TargetArray:Array of Strings=Get ProjectNames() -> (ProjectNames:Array of Strings)
    NewItem:"String (by ref)"=Get Name!(self:BP Animation Browser Project Object Reference=@n12) -> (Name:String)
   )
   > n13 = GetFocusedLevelSequence()
   > Set Level Sequence(Level Sequence=@n13)
  ]}
 FindCurrentToggle {nodes [
   FindCurrentToggle() {is=title K2Node_FunctionEntry}
   > n1 = Keys!(execute, TargetMap:Map of Strings to Booleans=Get isSubSequenceMute()) -> (then, Keys:Array of Strings)
   > n2 = For Each Loop(Array:Array of Strings=@n1) -> (Array Element:String)
   > Set CurrentKey(CurrentKey:String=@n2) -> (Output_Get:String)
   > Set CurrentIndex(CurrentIndex:int=@n2.Array Index) -> (Output_Get:int=0)
   > n3 = GetTracks!(execute, Sequence:Movie Scene Sequence Object Reference=Get Level Sequence()) -> (then, ReturnValue:Array of Movie Scene Track Object References)
   > n4 = For Each Loop(Array:Array of Movie Scene Track Object References=@n3) -> (Array Element:Movie Scene Track Object Reference)
   > n5 = GetDisplayName!(execute:Exec, Track:Movie Scene Track Object Reference=@n4) -> (then:Exec, ReturnValue:Text)
   > Branch(Condition=Contains(SearchIn="To String (Text)"(InText=@n5), Substring=Get CurrentKey() -> (CurrentKey:String)))
   > Add(TargetMap=Get ContainsSubSequence(), Key=Get CurrentKey(), Value=true) {is=title K2Node_CallFunction}
   > n6 = Cast To MovieSceneSubTrack!(execute, Object=@n4) -> (then, CastFailed, AsMovie Scene Sub Track:Movie Scene Sub Track Object Reference) {is=title K2Node_DynamicCast}
   > n7 = GetSections!(execute, Track:Movie Scene Track Object Reference=@n6) -> (then, ReturnValue:Array of Movie Scene Section Object References)
   > n8 = For Each Loop(Array:Array of Movie Scene Section Object References=@n7) -> (Array Element:Movie Scene Section Object Reference)
   > n9 = Branch(Condition=IsActive!(self:Movie Scene Section Object Reference=@n8) -> (ReturnValue:bool))
   n10 = "Get (a copy)"(Array:Array of Editor Utility Check Box Object References=Get CheckBoxes(), Dimension 1=Get CurrentIndex() -> (CurrentIndex:int=0)) -> (Output:Editor Utility Check Box Object Reference)
   > Set CheckedState(CheckedState=Checked, self=@n10) -> (Output_Get=Unchecked)
   > Add(TargetMap=Get isSubSequenceMute(), Key=Get CurrentKey(), Value=true)
   Set CheckedState(execute=@n9.else, CheckedState=Unchecked, self=@n10) -> (Output_Get=Unchecked)
   > Add(TargetMap=Get isSubSequenceMute(), Key=Get CurrentKey())
  ]}
 GetMetaData {nodes [
   GetMetaData() {is=title K2Node_FunctionEntry}
   > n1 = Sequence!(execute) -> (then_0, then_1, then_2:Exec, then_3:Exec)
   n2 = Get Level Sequence()
   > n3 = GetMetadataTag!(execute, Object:Object Reference=@n2, Tag:Name=Project:) -> (then, ReturnValue:String)
   > Set ProjectName(ProjectName:String=@n3) -> (Output_Get:String)
   n4 = GetMetadataTag(execute=@n1.then_1, Object=@n2, Tag=SubFolder:)
   > Set SubFolderName(SubFolderName:String=@n4) -> (Output_Get:String)
   n5 = GetMetadataTag(execute=@n1.then_2, Object=@n2, Tag=ShotNumber:)
   > Set ShotNumber(ShotNumber:String=@n5) -> (Output_Get:String)
   n6 = Get ProjectName()
   n7 = Branch(execute=@n1.then_3, Condition="integer > integer"(A=Len(S=@n6)))
   n8 = For Each Loop(Exec=@n7.else, Array:Array of Strings=Get ProjectNames()) -> (Array Element:String)
   n9 = Get Object Path String(Object=Get Level Sequence())
   n10 = RightChop!(SourceString:String=@n9, Count:int=int + int(A=FindSubstring(SearchIn=@n9, Substring=sequence), B=9)) -> (ReturnValue:String)
   Set SubFolderName(execute=@n8.Completed, SubFolderName=Left!(SourceString:String=@n10, Count:int=3) -> (ReturnValue:String))
   > Set ShotNumber(ShotNumber=RightChop(SourceString=Left(SourceString=@n10, Count=11), Count=7))
   Branch(execute=@n8.LoopBody, Condition=Contains(SearchIn=@n9, Substring=@n8))
   > Set ProjectName(ProjectName=@n8)
  ]}
 GetProjectAC {nodes [
   GetProjectAC() {is=title K2Node_FunctionEntry}
   n1 = Find Item!(TargetArray=Get ProjectNames(), ItemToFind:"String (by ref)"=Get ProjectName()) -> (ReturnValue)
   n2 = Get CurrentProjects() -> (CurrentProjects:Array of BP Animation Browser Project Object References)
   > n3 = For Each Loop(Array:Array of Animatable Creature Object References=Get Animatable Creatures!(
     Project:BP Animation Browser Project Object Reference="Get (a copy)"(Array:Array of BP Animation Browser Project Object References=@n2, Dimension 1=@n1) -> (Output:BP Animation Browser Project Object Reference)
     bSearchSubClasses:bool=false
     __WorldContext:Object Reference
    ) -> (Animatable Creatures:Array of Animatable Creature Object References) {is=title K2Node_CallFunction}) -> (Array Element:Animatable Creature Object Reference)
   > Set AnimatableCreature(AnimatableCreature:Animatable Creature Object Reference=@n3) -> (Output_Get:Animatable Creature Object Reference)
  ]}
 GetProjects {nodes [
   GetProjects() {is=title K2Node_FunctionEntry}
   > Clear(TargetArray:Array of BP Animation Browser Project Object References=Get Local Projects() -> (Local Projects:Array of BP Animation Browser Project Object References))
   > n1 = GetAssetsByClass!(
    execute
    self:Asset Registry Interface=GetAssetRegistry()
    ClassPathName:Top Level Asset Path Structure=GetSoftClassTopLevelAssetPath!(SoftClassReference:Object Soft Class Reference=/AnimInEdTools/AnimationBrowser/ProjectSettings/BP_AnimationBrowser_Project.BP_AnimationBrowser_Project_C) -> (ReturnValue:Top Level Asset Path Structure)
    bSearchSubClasses:bool=false
   ) -> (then, OutAssetData:Array of Asset Data Structures, ReturnValue:bool)
   > n2 = For Each Loop(Array:Array of Asset Data Structures=@n1) -> (Array Element:Asset Data Structure)
   Return Node!(execute=@n2.Completed, Projects:Array of BP Animation Browser Project Object References=Get Local Projects())
   n3 = Break Asset Data(AssetData=@n2)
   > n4 = LoadAsset(AssetPath=GetBaseFilename(InPath="To String (Name)"(InName=@n3), bRemovePath=false))
   > n5 = Cast To BP_AnimationBrowser_Project!(execute, Object=@n4) -> (then, CastFailed, AsBP Animation Browser Project:BP Animation Browser Project Object Reference) {is=title K2Node_DynamicCast}
   > Add(
    TargetArray:Array of BP Animation Browser Project Object References=Get Local Projects()
    NewItem:"BP Animation Browser Project Object Reference (by ref)"=@n5
   ) {is=title K2Node_CallArrayFunction}
  ]}
 SetTracks {nodes [
   SetTracks() {is=title K2Node_FunctionEntry}
   > n1 = "For Each Loop (Map)"!(execute:Exec, MapPin:Map of Strings to Booleans=Get isSubSequenceMute(), BreakPin:Exec) -> (then:Exec, KeyPin:String, ValuePin:bool, CompletedPin:Exec)
   > Set Key(Key:String=@n1) -> (Output_Get:String)
   > n2 = Set Value(Value:bool=@n1.ValuePin) -> (Output_Get:bool)
   > n3 = Branch(Condition=@n2)
   > n4 = Branch(Condition=Find(TargetMap:Map of Strings to Booleans=Get ContainsSubSequence(), Key=Get Key() -> (Key:String)) -> (Value:bool))
   n5 = GetTracks(execute=[@n4.then, @n3.else], Sequence=Get Level Sequence())
   > n6 = For Each Loop(Array:Array of Movie Scene Track Object References=@n5) -> (Array Element:Movie Scene Track Object Reference)
   > n7 = GetDisplayName(Track=@n6)
   > Branch(Condition=Contains(SearchIn="To String (Text)"(InText=@n7), Substring=Get Key()))
   > n8 = Cast To MovieSceneSubTrack(Object=@n6)
   > n9 = GetSections(Track=@n8)
   > n10 = For Each Loop(Array:Array of Movie Scene Section Object References=@n9) -> (Array Element:Movie Scene Section Object Reference)
   > SetIsActive!(execute, self:Movie Scene Section Object Reference=@n10, bInIsActive:bool=Get Value() -> (Value:bool)) -> (then)
   AddMissingSubSequence!(execute=@n4.else, self, MissingSubSequence:String=Get Key()) -> (then) {is=title K2Node_CallFunction}
  ]}
]
variables [
 AnimatableCreature {type=AnimatableCreature_C, default=null}
 CheckBoxes {type="array<EditorUtilityCheckBox>", default []}
 ContainsSubSequence {type="map<boolean>", default {Cam=false, aud=false, gam=false, lgt=false, tec=false, vfx=false}}
 Current Binding {type=MovieSceneBindingProxy, default {bindingId=#1, sequence=null}}
 Current Project {type=BP_AnimationBrowser_Project_C, default=null}
 CurrentProjects {type="array<BP_AnimationBrowser_Project_C>", default []}
 isSubSequenceMute {type="map<boolean>", default {Cam=false, aud=false, gam=false, lgt=false, tec=false, vfx=false}}
 Level Sequence {type=LevelSequence, default=null}
 ProjectName {type=string, default=""}
 ProjectNames {type="array<string>", default []}
 ShotNumber {type=string, default=""}
 SubFolderName {type=string, default=""}
]
widgets {
 info {rootWidgetClass=&CanvasPanel, widgetCount=24}
 widgets [
  CanvasPanel_0 {widgetClassPath=&CanvasPanel, widget=CanvasPanel_0}
  VerticalBox_0 {
   widgetClassPath=&VerticalBox
   parent=CanvasPanel_0
   widget=VerticalBox_0
   slot {class=CanvasPanelSlot, properties {layoutData {alignment=<0 0>, anchors {maximum=<1 1>, minimum=<0 0>}, offsets=<ltrb 0 0 0 0>}}}
  }
  HorizontalBox_0 {widgetClassPath=&HorizontalBox, parent=VerticalBox_0, widget=HorizontalBox_0, slot {class=VerticalBoxSlot}}
  LevelSequencePropertyView {
   widgetClassPath=&SinglePropertyView
   parent=HorizontalBox_0
   widget=LevelSequencePropertyView
   properties {propertyName=Level Sequence}
   slot {class=HorizontalBoxSlot, properties {size {sizeRule=Fill, value=2.5}}}
  }
  CurrentLevelSequenceButton {
   widgetClassPath=&EditorUtilityButton
   parent=HorizontalBox_0
   widget=CurrentLevelSequenceButton
   slot {class=HorizontalBoxSlot}
  }
  TextBlock_106 {
   widgetClassPath=&TextBlock
   parent=CurrentLevelSequenceButton
   widget=TextBlock_106
   properties {
    text=Get Current LS
    font {
     bForceMonospaced=false
     bMaterialIsStencil=false
     fontMaterial=null
     fontObject=&Roboto
     letterSpacing=0
     monospacedWidth=1
     size=13.5
     skewAmount=0
     typefaceFontName=Bold
     outlineSettings {
      bApplyOutlineToDropShadows=false
      bMiteredCorners=false
      bSeparateFillAlpha=false
      outlineMaterial=null
      outlineSize=0
      outlineColor=<rgba 0 0 0 1>
     }
    }
   }
   slot {class=ButtonSlot}
  }
  Spacer_221 {widgetClassPath=&Spacer, parent=VerticalBox_0, widget=Spacer_221, properties {size=<1 15>}, slot {class=VerticalBoxSlot}}
  HorizontalBox {widgetClassPath=&HorizontalBox, parent=VerticalBox_0, widget=HorizontalBox, slot {class=VerticalBoxSlot}}
  AudioCheckBox {
   widgetClassPath=&EditorUtilityCheckBox
   bIsVariable=true
   parent=HorizontalBox
   widget=AudioCheckBox
   slot {class=HorizontalBoxSlot, properties {horizontalAlignment=HAlign_Center, verticalAlignment=VAlign_Center, size {sizeRule=Fill, value=1}}}
  }
  TextBlock_97 {
   widgetClassPath=&TextBlock
   parent=AudioCheckBox
   widget=TextBlock_97
   properties {
    text=Audio
    font {
     bForceMonospaced=false
     bMaterialIsStencil=false
     fontMaterial=null
     fontObject=&Roboto
     letterSpacing=0
     monospacedWidth=1
     size=12
     skewAmount=0
     typefaceFontName=Bold
     outlineSettings {
      bApplyOutlineToDropShadows=false
      bMiteredCorners=false
      bSeparateFillAlpha=false
      outlineMaterial=null
      outlineSize=0
      outlineColor=<rgba 0 0 0 1>
     }
    }
   }
   slot {class=PanelSlot}
  }
  CameraCheckBox {
   widgetClassPath=&EditorUtilityCheckBox
   bIsVariable=true
   parent=HorizontalBox
   widget=CameraCheckBox
   slot {class=HorizontalBoxSlot, properties {horizontalAlignment=HAlign_Center, verticalAlignment=VAlign_Center, size {sizeRule=Fill, value=1}}}
  }
  TextBlock {
   widgetClassPath=&TextBlock
   parent=CameraCheckBox
   widget=TextBlock
   properties {
    text=Camera
    font {
     bForceMonospaced=false
     bMaterialIsStencil=false
     fontMaterial=null
     fontObject=&Roboto
     letterSpacing=0
     monospacedWidth=1
     size=12
     skewAmount=0
     typefaceFontName=Bold
     outlineSettings {
      bApplyOutlineToDropShadows=false
      bMiteredCorners=false
      bSeparateFillAlpha=false
      outlineMaterial=null
      outlineSize=0
      outlineColor=<rgba 0 0 0 1>
     }
    }
   }
   slot {class=PanelSlot}
  }
  GamePlayCheckBox {
   widgetClassPath=&EditorUtilityCheckBox
   bIsVariable=true
   parent=HorizontalBox
   widget=GamePlayCheckBox
   slot {class=HorizontalBoxSlot, properties {horizontalAlignment=HAlign_Center, verticalAlignment=VAlign_Center, size {sizeRule=Fill, value=1}}}
  }
  TextBlock_1 {
   widgetClassPath=&TextBlock
   parent=GamePlayCheckBox
   widget=TextBlock_1
   properties {
    text=Gameplay
    font {
     bForceMonospaced=false
     bMaterialIsStencil=false
     fontMaterial=null
     fontObject=&Roboto
     letterSpacing=0
     monospacedWidth=1
     size=12
     skewAmount=0
     typefaceFontName=Bold
     outlineSettings {
      bApplyOutlineToDropShadows=false
      bMiteredCorners=false
      bSeparateFillAlpha=false
      outlineMaterial=null
      outlineSize=0
      outlineColor=<rgba 0 0 0 1>
     }
    }
   }
   slot {class=PanelSlot}
  }
  LightingCheckBox {
   widgetClassPath=&EditorUtilityCheckBox
   bIsVariable=true
   parent=HorizontalBox
   widget=LightingCheckBox
   slot {class=HorizontalBoxSlot, properties {horizontalAlignment=HAlign_Center, verticalAlignment=VAlign_Center, size {sizeRule=Fill, value=1}}}
  }
  TextBlock_2 {
   widgetClassPath=&TextBlock
   parent=LightingCheckBox
   widget=TextBlock_2
   properties {
    text=Lighting
    font {
     bForceMonospaced=false
     bMaterialIsStencil=false
     fontMaterial=null
     fontObject=&Roboto
     letterSpacing=0
     monospacedWidth=1
     size=12
     skewAmount=0
     typefaceFontName=Bold
     outlineSettings {
      bApplyOutlineToDropShadows=false
      bMiteredCorners=false
      bSeparateFillAlpha=false
      outlineMaterial=null
      outlineSize=0
      outlineColor=<rgba 0 0 0 1>
     }
    }
   }
   slot {class=PanelSlot}
  }
  TechArtCheckBox {
   widgetClassPath=&EditorUtilityCheckBox
   bIsVariable=true
   parent=HorizontalBox
   widget=TechArtCheckBox
   slot {class=HorizontalBoxSlot, properties {horizontalAlignment=HAlign_Center, verticalAlignment=VAlign_Center, size {sizeRule=Fill, value=1}}}
  }
  TextBlock_3 {
   widgetClassPath=&TextBlock
   parent=TechArtCheckBox
   widget=TextBlock_3
   properties {
    text=TechArt
    font {
     bForceMonospaced=false
     bMaterialIsStencil=false
     fontMaterial=null
     fontObject=&Roboto
     letterSpacing=0
     monospacedWidth=1
     size=12
     skewAmount=0
     typefaceFontName=Bold
     outlineSettings {
      bApplyOutlineToDropShadows=false
      bMiteredCorners=false
      bSeparateFillAlpha=false
      outlineMaterial=null
      outlineSize=0
      outlineColor=<rgba 0 0 0 1>
     }
    }
   }
   slot {class=PanelSlot}
  }
  VFXCheckBox {
   widgetClassPath=&EditorUtilityCheckBox
   bIsVariable=true
   parent=HorizontalBox
   widget=VFXCheckBox
   slot {class=HorizontalBoxSlot, properties {horizontalAlignment=HAlign_Center, verticalAlignment=VAlign_Center, size {sizeRule=Fill, value=1}}}
  }
  TextBlock_5 {
   widgetClassPath=&TextBlock
   parent=VFXCheckBox
   widget=TextBlock_5
   properties {
    text=VFX
    font {
     bForceMonospaced=false
     bMaterialIsStencil=false
     fontMaterial=null
     fontObject=&Roboto
     letterSpacing=0
     monospacedWidth=1
     size=12
     skewAmount=0
     typefaceFontName=Bold
     outlineSettings {
      bApplyOutlineToDropShadows=false
      bMiteredCorners=false
      bSeparateFillAlpha=false
      outlineMaterial=null
      outlineSize=0
      outlineColor=<rgba 0 0 0 1>
     }
    }
   }
   slot {class=PanelSlot}
  }
  Spacer_93 {widgetClassPath=&Spacer, parent=HorizontalBox, widget=Spacer_93, slot {class=HorizontalBoxSlot}}
  ScrollBox {
   widgetClassPath=&ScrollBox
   bIsVariable=true
   parent=VerticalBox_0
   widget=ScrollBox
   properties {
    consumeMouseWheel=Always
    widgetBarStyle {
     thickness=8
     draggedThumbImage {
      bIsDynamicallyLoaded=false
      drawAs=RoundedBox
      imageType=NoImage
      mirroring=NoMirror
      resourceName=None
      resourceObject=null
      tiling=NoTile
      imageSize=<0 0>
      margin=<ltrb 0 0 0 0>
      outlineSettings {
       bUseBrushTransparency=false
       roundingType=FixedRadius
       width=0
       color {colorUseRule=UseColor_Specified, specifiedColor=<rgba 0 0 0 0>}
       cornerRadii=<4 4 4 4>
      }
      tintColor {colorUseRule=UseColor_Specified, specifiedColor=<rgba 0.2159 0.2159 0.2159 1>}
      uVRegion {bIsValid=false, max=<0 0>, min=<0 0>}
     }
     horizontalBackgroundImage {
      bIsDynamicallyLoaded=false
      drawAs=NoDrawType
      imageType=NoImage
      mirroring=NoMirror
      resourceName=None
      resourceObject=null
      tiling=NoTile
      imageSize=<0 0>
      margin=<ltrb 0 0 0 0>
      outlineSettings {
       bUseBrushTransparency=false
       roundingType=HalfHeightRadius
       width=0
       color {colorUseRule=UseColor_Specified, specifiedColor=<rgba 0 0 0 0>}
       cornerRadii=<0 0 0 0>
      }
      tintColor {colorUseRule=UseColor_Specified, specifiedColor=<rgba 1 1 1 1>}
      uVRegion {bIsValid=false, max=<0 0>, min=<0 0>}
     }
     horizontalBottomSlotImage {
      bIsDynamicallyLoaded=false
      drawAs=NoDrawType
      imageType=NoImage
      mirroring=NoMirror
      resourceName=None
      resourceObject=null
      tiling=NoTile
      imageSize=<0 0>
      margin=<ltrb 0 0 0 0>
      outlineSettings {
       bUseBrushTransparency=false
       roundingType=HalfHeightRadius
       width=0
       color {colorUseRule=UseColor_Specified, specifiedColor=<rgba 0 0 0 0>}
       cornerRadii=<0 0 0 0>
      }
      tintColor {colorUseRule=UseColor_Specified, specifiedColor=<rgba 1 1 1 1>}
      uVRegion {bIsValid=false, max=<0 0>, min=<0 0>}
     }
     horizontalTopSlotImage {
      bIsDynamicallyLoaded=false
      drawAs=NoDrawType
      imageType=NoImage
      mirroring=NoMirror
      resourceName=None
      resourceObject=null
      tiling=NoTile
      imageSize=<0 0>
      margin=<ltrb 0 0 0 0>
      outlineSettings {
       bUseBrushTransparency=false
       roundingType=HalfHeightRadius
       width=0
       color {colorUseRule=UseColor_Specified, specifiedColor=<rgba 0 0 0 0>}
       cornerRadii=<0 0 0 0>
      }
      tintColor {colorUseRule=UseColor_Specified, specifiedColor=<rgba 1 1 1 1>}
      uVRegion {bIsValid=false, max=<0 0>, min=<0 0>}
     }
     hoveredThumbImage {
      bIsDynamicallyLoaded=false
      drawAs=RoundedBox
      imageType=NoImage
      mirroring=NoMirror
      resourceName=None
      resourceObject=null
      tiling=NoTile
      imageSize=<0 0>
      margin=<ltrb 0 0 0 0>
      outlineSettings {
       bUseBrushTransparency=false
       roundingType=FixedRadius
       width=0
       color {colorUseRule=UseColor_Specified, specifiedColor=<rgba 0 0 0 0>}
       cornerRadii=<4 4 4 4>
      }
      tintColor {colorUseRule=UseColor_Specified, specifiedColor=<rgba 0.2159 0.2159 0.2159 1>}
      uVRegion {bIsValid=false, max=<0 0>, min=<0 0>}
     }
     normalThumbImage {
      bIsDynamicallyLoaded=false
      drawAs=RoundedBox
      imageType=NoImage
      mirroring=NoMirror
      resourceName=None
      resourceObject=null
      tiling=NoTile
      imageSize=<0 0>
      margin=<ltrb 0 0 0 0>
      outlineSettings {
       bUseBrushTransparency=false
       roundingType=FixedRadius
       width=0
       color {colorUseRule=UseColor_Specified, specifiedColor=<rgba 0 0 0 0>}
       cornerRadii=<4 4 4 4>
      }
      tintColor {colorUseRule=UseColor_Specified, specifiedColor=<rgba 0.0953 0.0953 0.0953 1>}
      uVRegion {bIsValid=false, max=<0 0>, min=<0 0>}
     }
     verticalBackgroundImage {
      bIsDynamicallyLoaded=false
      drawAs=NoDrawType
      imageType=NoImage
      mirroring=NoMirror
      resourceName=None
      resourceObject=null
      tiling=NoTile
      imageSize=<0 0>
      margin=<ltrb 0 0 0 0>
      outlineSettings {
       bUseBrushTransparency=false
       roundingType=HalfHeightRadius
       width=0
       color {colorUseRule=UseColor_Specified, specifiedColor=<rgba 0 0 0 0>}
       cornerRadii=<0 0 0 0>
      }
      tintColor {colorUseRule=UseColor_Specified, specifiedColor=<rgba 1 1 1 1>}
      uVRegion {bIsValid=false, max=<0 0>, min=<0 0>}
     }
     verticalBottomSlotImage {
      bIsDynamicallyLoaded=false
      drawAs=NoDrawType
      imageType=NoImage
      mirroring=NoMirror
      resourceName=None
      resourceObject=null
      tiling=NoTile
      imageSize=<0 0>
      margin=<ltrb 0 0 0 0>
      outlineSettings {
       bUseBrushTransparency=false
       roundingType=HalfHeightRadius
       width=0
       color {colorUseRule=UseColor_Specified, specifiedColor=<rgba 0 0 0 0>}
       cornerRadii=<0 0 0 0>
      }
      tintColor {colorUseRule=UseColor_Specified, specifiedColor=<rgba 1 1 1 1>}
      uVRegion {bIsValid=false, max=<0 0>, min=<0 0>}
     }
     verticalTopSlotImage {
      bIsDynamicallyLoaded=false
      drawAs=NoDrawType
      imageType=NoImage
      mirroring=NoMirror
      resourceName=None
      resourceObject=null
      tiling=NoTile
      imageSize=<0 0>
      margin=<ltrb 0 0 0 0>
      outlineSettings {
       bUseBrushTransparency=false
       roundingType=HalfHeightRadius
       width=0
       color {colorUseRule=UseColor_Specified, specifiedColor=<rgba 0 0 0 0>}
       cornerRadii=<0 0 0 0>
      }
      tintColor {colorUseRule=UseColor_Specified, specifiedColor=<rgba 1 1 1 1>}
      uVRegion {bIsValid=false, max=<0 0>, min=<0 0>}
     }
    }
    widgetStyle {
     barThickness=8
     bottomShadowBrush {
      bIsDynamicallyLoaded=false
      drawAs=Box
      imageType=FullColor
      mirroring=NoMirror
      resourceName=../../../Engine/Content/Slate/Common/ScrollBoxShadowBottom.png
      resourceObject=null
      tiling=NoTile
      imageSize=<16 8>
      margin=<ltrb 0.5 0 0.5 1>
      outlineSettings {
       bUseBrushTransparency=false
       roundingType=HalfHeightRadius
       width=0
       color {colorUseRule=UseColor_Specified, specifiedColor=<rgba 0 0 0 0>}
       cornerRadii=<0 0 0 0>
      }
      tintColor {colorUseRule=UseColor_Specified, specifiedColor=<rgba 1 1 1 1>}
      uVRegion {bIsValid=false, max=<0 0>, min=<0 0>}
     }
     horizontalScrolledContentPadding=<ltrb 0 0 1 0>
     leftShadowBrush {
      bIsDynamicallyLoaded=false
      drawAs=Box
      imageType=FullColor
      mirroring=NoMirror
      resourceName=../../../Engine/Content/Slate/Common/ScrollBoxShadowLeft.png
      resourceObject=null
      tiling=NoTile
      imageSize=<8 16>
      margin=<ltrb 1 0.5 0 0.5>
      outlineSettings {
       bUseBrushTransparency=false
       roundingType=HalfHeightRadius
       width=0
       color {colorUseRule=UseColor_Specified, specifiedColor=<rgba 0 0 0 0>}
       cornerRadii=<0 0 0 0>
      }
      tintColor {colorUseRule=UseColor_Specified, specifiedColor=<rgba 1 1 1 1>}
      uVRegion {bIsValid=false, max=<0 0>, min=<0 0>}
     }
     rightShadowBrush {
      bIsDynamicallyLoaded=false
      drawAs=Box
      imageType=FullColor
      mirroring=NoMirror
      resourceName=../../../Engine/Content/Slate/Common/ScrollBoxShadowRight.png
      resourceObject=null
      tiling=NoTile
      imageSize=<8 16>
      margin=<ltrb 0 0.5 1 0.5>
      outlineSettings {
       bUseBrushTransparency=false
       roundingType=HalfHeightRadius
       width=0
       color {colorUseRule=UseColor_Specified, specifiedColor=<rgba 0 0 0 0>}
       cornerRadii=<0 0 0 0>
      }
      tintColor {colorUseRule=UseColor_Specified, specifiedColor=<rgba 1 1 1 1>}
      uVRegion {bIsValid=false, max=<0 0>, min=<0 0>}
     }
     topShadowBrush {
      bIsDynamicallyLoaded=false
      drawAs=Box
      imageType=FullColor
      mirroring=NoMirror
      resourceName=../../../Engine/Content/Slate/Common/ScrollBoxShadowTop.png
      resourceObject=null
      tiling=NoTile
      imageSize=<16 8>
      margin=<ltrb 0.5 1 0.5 0>
      outlineSettings {
       bUseBrushTransparency=false
       roundingType=HalfHeightRadius
       width=0
       color {colorUseRule=UseColor_Specified, specifiedColor=<rgba 0 0 0 0>}
       cornerRadii=<0 0 0 0>
      }
      tintColor {colorUseRule=UseColor_Specified, specifiedColor=<rgba 1 1 1 1>}
      uVRegion {bIsValid=false, max=<0 0>, min=<0 0>}
     }
     verticalScrolledContentPadding=<ltrb 0 0 0 1>
    }
   }
   slot {class=VerticalBoxSlot, properties {size {sizeRule=Fill, value=1}}}
  }
  RunButton {
   widgetClassPath=&Button
   parent=VerticalBox_0
   widget=RunButton
   properties {widgetStyle {
     clickedSlateSound {resourceObject=null}
     disabled {
      bIsDynamicallyLoaded=false
      drawAs=RoundedBox
      imageType=NoImage
      mirroring=NoMirror
      resourceName=None
      resourceObject=null
      tiling=NoTile
      imageSize=<0 0>
      margin=<ltrb 0 0 0 0>
      outlineSettings {
       bUseBrushTransparency=false
       roundingType=FixedRadius
       width=1
       color {colorUseRule=UseColor_ColorTable, specifiedColor=<rgba 0 0 0 0>}
       cornerRadii=<4 4 4 4>
      }
      tintColor {colorUseRule=UseColor_Specified, specifiedColor=<rgba 0.0395 0.0395 0.0395 1>}
      uVRegion {bIsValid=false, max=<0 0>, min=<0 0>}
     }
     disabledForeground {colorUseRule=UseColor_ColorTable, specifiedColor=<rgba 0 0 0 0>}
     hovered {
      bIsDynamicallyLoaded=false
      drawAs=RoundedBox
      imageType=NoImage
      mirroring=NoMirror
      resourceName=None
      resourceObject=null
      tiling=NoTile
      imageSize=<32 32>
      margin=<ltrb 0 0 0 0>
      outlineSettings {
       bUseBrushTransparency=false
       roundingType=FixedRadius
       width=1
       color {colorUseRule=UseColor_ColorTable, specifiedColor=<rgba 0 0 0 0>}
       cornerRadii=<4 4 4 4>
      }
      tintColor {colorUseRule=UseColor_Specified, specifiedColor=<rgba 0.0953 0.0953 0.0953 1>}
      uVRegion {bIsValid=false, max=<0 0>, min=<0 0>}
     }
     hoveredForeground {colorUseRule=UseColor_ColorTable, specifiedColor=<rgba 0 0 0 0>}
     hoveredSlateSound {resourceObject=null}
     normal {
      bIsDynamicallyLoaded=false
      drawAs=RoundedBox
      imageType=NoImage
      mirroring=NoMirror
      resourceName=None
      resourceObject=null
      tiling=NoTile
      imageSize=<32 32>
      margin=<ltrb 0 0 0 0>
      outlineSettings {
       bUseBrushTransparency=false
       roundingType=FixedRadius
       width=1
       color {colorUseRule=UseColor_ColorTable, specifiedColor=<rgba 0 0 0 0>}
       cornerRadii=<4 4 4 4>
      }
      tintColor {colorUseRule=UseColor_Specified, specifiedColor=<rgba 0.0395 0.0395 0.0395 1>}
      uVRegion {bIsValid=false, max=<0 0>, min=<0 0>}
     }
     normalForeground {colorUseRule=UseColor_ColorTable, specifiedColor=<rgba 0 0 0 0>}
     normalPadding=<ltrb 12 1.5 12 1.5>
     pressed {
      bIsDynamicallyLoaded=false
      drawAs=RoundedBox
      imageType=NoImage
      mirroring=NoMirror
      resourceName=None
      resourceObject=null
      tiling=NoTile
      imageSize=<32 32>
      margin=<ltrb 0 0 0 0>
      outlineSettings {
       bUseBrushTransparency=false
       roundingType=FixedRadius
       width=1
       color {colorUseRule=UseColor_ColorTable, specifiedColor=<rgba 0 0 0 0>}
       cornerRadii=<4 4 4 4>
      }
      tintColor {colorUseRule=UseColor_Specified, specifiedColor=<rgba 0.0284 0.0284 0.0284 1>}
      uVRegion {bIsValid=false, max=<0 0>, min=<0 0>}
     }
     pressedForeground {colorUseRule=UseColor_ColorTable, specifiedColor=<rgba 0 0 0 0>}
     pressedPadding=<ltrb 12 2.5 12 0.5>
     pressedSlateSound {resourceObject=null}
    }}
   slot {class=VerticalBoxSlot}
  }
  TextBlock_4 {widgetClassPath=&TextBlock, parent=RunButton, widget=TextBlock_4, properties {text=Set Sub Tracks}, slot {class=ButtonSlot}}
 ]
}
